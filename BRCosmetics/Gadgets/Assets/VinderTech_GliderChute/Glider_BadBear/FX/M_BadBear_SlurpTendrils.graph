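# Translucent subsurface material for slurp liquid tendrils on the Bad Bear glider, usable on Niagara and skeletal meshes.
# MF_Slurp_forFX supplies surface attributes, with emissive scaled by a dynamic parameter and opacity using particle alpha and depth fade.
# Niagara-driven WPO combines vertex-red-weighted bend, panning-noise normal displacement and another offset.
# Panning noise-masked normals, faded along UV1, optionally blend to a texture normal.
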Material /BRCosmetics/Gadgets/Assets/VinderTech_GliderChute/Glider_BadBear/FX/M_BadBear_SlurpTendrils
graph {
 properties {
  bUsedWithMeshDeformer=true
  bUsedWithNiagaraMeshParticles=true
  bUsedWithSkeletalMesh=true
  blendMode=BLEND_Translucent
  opacityMaskClipValue=0.1333
  pixelDepthOffsetMode=PDOM_Legacy
  shadingModel=MSM_Subsurface
  translucencyLightingMode=TLM_Surface
  usedShadingModels=Subsurface
 }
 nodes [
  n1 = DynamicParameter() {desc=Parameters set by Niagara, paramNames [X,Y,Z,WPOIntensity]}
  n2 = OneMinus(None=VertexColor().R)
  n3 = ScalarParameter() {defaultValue=3, group=WPO, parameterName=InterpolateWPO_BendPower}
  n4 = LinearInterpolate(
   B=Clamp(
    None=MaterialFunctionCall(X=@n1.X, Y=@n1.X, Z=@n1.X) {materialFunction=&MakeFloat3}.Result
    Min=ScalarParameter() {defaultValue=-1000, group=WPO, parameterName=InterpolateWPO_Clamp_min}
    Max=ScalarParameter() {defaultValue=1000, group=WPO, parameterName=InterpolateWPO_Clamp_max}
   ) {maxDefault=1000, minDefault=-1000}
   Alpha=Multiply(A=Power(Base=@n2, Exp=@n3), B=@n1.WPOIntensity)
  )
  n5 = VertexNormalWS()
  n6 = DynamicParameter() {parameterIndex=1, paramNames [Min WPO, Max WPO, X Speed, Y Speed]}
  n7 = AppendVector(A=@n6.X Speed, B=@n6.X Speed)
  n8 = LinearInterpolate(
   A=@n6.Min WPO
   B=@n6.Min WPO
   Alpha=TextureSample(UVs=Panner(Speed=@n7) {speedY=2}) {texture=/Game/Effects/Fort_Effects/Textures/Power_Source/T_Power_Source_Sphere_Texture}.R
  ) {constA=-1}
  Add(A=Add(A=Transform(None=@n4) {transformSourceType=TRANSFORMSOURCE_Local}, B=Multiply(A=@n5, B=@n8)), B=NamedRerouteUsage())
  ComponentMask(None=OneMinus(None=TextureCoordinate())) {g=true}
  n9 = NamedRerouteUsage()
  n10 = GetMaterialAttributes(None=MaterialFunctionCall!(Pre Skinned Position, World Normal=Transform(None=@n9), AO) {materialFunction=/Game/Effects/Niagara/MasterMaterials/Liquid/MF_Slurp_forFX}.Result) {attributeGetTypes [
    #1
    #2
    #3
    #4
    #5
    #6
   ]}
  n11 = Multiply(A=@n10.Emissive Color, B=DynamicParameter() {parameterIndex=1, paramNames [Min WPO, Max WPO, X Speed, Y Speed]}.X Speed) {constB=0.05}
  MaterialFunctionCall(Emissive=@n11) {materialFunction=&MF_FX_EmissiveAdjustments}
  Multiply(A=ParticleColor().A, B=DepthFade(FadeDistance=ScalarParameter() {defaultValue=75, parameterName=Fade Distance}))
  NamedRerouteDeclaration(None=@n10.World Position Offset) {name=WPO, nodeColor=<rgba 0.4297 0 1 1>}
  NamedRerouteDeclaration(None=@n10.Roughness) {name=Roughness, nodeColor=<rgba 0.2891 0 1 1>}
  NamedRerouteDeclaration(None=@n10.Base Color) {name=Color, nodeColor=<rgba 1 0 0.0703 1>}
  NamedRerouteDeclaration(None=@n11) {name=Emissive, nodeColor=<rgba 1 0.3984 0 1>}
  n12 = Constant3Vector() {constant=<rgba 0 0 1 0>}
  n13 = TextureCoordinate() {coordinateIndex=1, uTiling=1.5, vTiling=1.5}
  n14 = ComponentMask(None=VectorParameter() {parameterName=Normal UV, defaultValue=<rgba 2 2 0 0>}.RGB) {g=true, r=true}
  n15 = AppendVector(A=Constant(), B=VertexColor().R)
  n16 = Panner(
   Coordinate=Add(A=Multiply(A=@n13, B=@n14), B=@n15)
   Speed=VectorParameter() {parameterName=Normal Pan Speed, defaultValue=<rgba 0 0.1 0 1>}.RGB
  ) {speedY=2}
  n17 = ComponentMask(None=NamedRerouteDeclaration(None=TextureSample(UVs=@n16) {samplerType=SAMPLERTYPE_Normal, texture=/BRCosmetics/Characters/Player/Male/Large/Bodies/M_LRG_Clash_V/Textures/Simple_Perlin_N}.RGB) {name=Normal, nodeColor=<rgba 1 0.4922 0 1>}) {g=true, r=true}
  n18 = Panner(Coordinate=Add(A=@n15, B=TextureCoordinate() {uTiling=2, vTiling=2})) {speedY=-0.3}
  n19 = Saturate(None=LinearInterpolate(Alpha=TextureSample(UVs=@n18) {texture=&T_Power_Source_Sphere_Texture}.R) {constA=-0.2, constB=5})
  n20 = MaterialFunctionCall(
   Normal=AppendVector(A=Multiply(A=@n17, B=@n19), B=Constant() {r=1})
   Flatness=ScalarParameter() {defaultValue=0.9, group=Normal, parameterName=Flatness}
  ) {materialFunction=&FlattenNormal}
  n21 = ComponentMask(None=TextureCoordinate() {coordinateIndex=1}) {g=true}
  n22 = Saturate(None=Multiply(A=Power(Base=@n21)) {constB=10})
  n23 = TextureSampleParameter2D(UVs=NamedRerouteUsage()) {
   group=Normal
   parameterName=Normal
   samplerType=SAMPLERTYPE_Normal
   texture=/Game/Athena/Apollo/Water/Textures/02_Rounded/02_WaterFoam_Veroni_base_N
  }
  NamedRerouteDeclaration(None=LinearInterpolate(A=LinearInterpolate(A=@n12, B=@n20.Result, Alpha=@n22), B=@n23.RGB, Alpha=ScalarParameter() {group=Normal, parameterName=Use_Normal_Texture})) {name=Normal_1, nodeColor=<rgba 0 0.9766 1 1>}
  NamedRerouteDeclaration(None=@n10.Specular) {name=Specular, nodeColor=<rgba 0 0.2031 1 1>}
  NamedRerouteDeclaration(None=@n10.Metallic) {name=Metalic, nodeColor=<rgba 0 0.8594 1 1>}
  NamedRerouteUsage()
  NamedRerouteUsage()
  NamedRerouteUsage()
  NamedRerouteUsage()
  NamedRerouteUsage()
  NamedRerouteUsage()
  NamedRerouteUsage()
 ]
}
