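# Niagara effect for the Bad Bear glider, layering goop drips, tendrils, speedlines and a wind cone.
# User parameters for deployment, front-end preview and forward/right motion dots gate spawning, flip directions and drive fades.
# Drips spawn on the glider skeletal mesh; six tendril meshes track its sockets or bones by spawn group, stretching and rewinding on a timeline.
# GPU speedlines spawn per distance, or by rate in preview; the wind cone mesh flips with forward sign.
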NiagaraSystem /BRCosmetics/Gadgets/Assets/VinderTech_GliderChute/Glider_BadBear/FX/NS_Glider_BadBear
dependencies {
 usedDataInterfaces [&NiagaraDataInterfaceCurve, &NiagaraDataInterfaceVector2DCurve]
 usedDynamicInputs []
 usedModules [&SystemState]
}
emitters {
 Drips {
  emitterUpdateScript [
   &EmitterState()
   &SpawnRate(
    Spawn Probability=0.5
    SpawnRate=&MakeCustomFloatFromBool(
     Bool=$User.bIsFullyDeployed
     True Float=&MakeCustomFloatFromBool(
      Bool=&SetBoolByFloatComparison(A=&Abs_Float(Float:NiagaraFloat=$User.ForwardDot), B=0.1, Comparison Type=::NewEnumerator2)
      True Float=4
      False Float=0
     )
     False Float=0
    )
    Use Spawn Probability=true
   )
  ]
  particleSpawnScript [
   &InitializeParticle(
    Lifetime Max=1.5
    Lifetime Min=0.5
    Lifetime Mode=::NewEnumerator1
    Sprite Size Mode=::NewEnumerator1
    Uniform Sprite Size Max=12
    Uniform Sprite Size Min=4
    UsePositionOffset=false
    Write Material Random=true
   )
   &SkeletalMeshLocation(
    MaskSpawnGroup=false
    Mesh Sampling Type=::NewEnumerator5
    OverrideRandomSeed=false
    Skeletal Mesh=$User.GliderMesh
    Velocity Sampling=::NewEnumerator0
    VelocitySpeedLimit=false
   )
   &ShapeLocation(
    MaskSpawnGroup=false
    Offset=<0 25 0>
    Offset Mode=::NewEnumerator0
    Override Seed=false
    Rotation Axis=<0 1 0>
    Rotation Mode=::NewEnumerator1
    Shape Origin=$Particles.Position
    Sphere Radius=3
    UseOwnerScale=true
   )
   &CameraOffset(Camera Offset Amount=30)
   &AddVelocity(
    Velocity=&MakeCustomVectorFromBool(
     Bool=&Boolean_OrOperation(
      Bool1=&SetBoolByFloatComparison(A=$User.ForwardDot, B=0.05, Comparison Type=::NewEnumerator2)
      Bool2=$User.bIsFrontEndPreview
     )
     True Vector=<0 1 0>
     False Vector=<0 -1 0>
    )
    Velocity Speed Scale=750
   )
   &AddVelocity(Velocity=$Particles.ShapeLocation.ShapeNormal, Velocity Speed Scale=150) {module=AddVelocity001}
  ]
  particleUpdateScript [
   &ParticleState()
   &ScaleSpriteSize(Uniform Curve Sprite Scale=DataInterface() {propertyValues {Curve {keys [{arriveTangent=-4.5487, interpMode=RCIM_Cubic, leaveTangent=-4.5487, tangentMode=RCTM_User, time=0.5, value=1}, <at 1 0 RCIM_Cubic>]}}})
   &GravityForce(Gravity=&Multiply_VectorByFloat(
     Vector=<0 0 -1600>
     Float=&FloatFromCurve(FloatCurve=DataInterface() {propertyValues {Curve {keys [{interpMode=RCIM_Cubic, leaveTangentWeight=0.6722, tangentMode=RCTM_User, time=0, value=0}, <at 1 1 RCIM_Cubic>]}}})
    ))
   &CurlNoiseForce(
    Noise Frequency=0.05
    Noise Quality / Cost=::NewEnumerator0
    Noise Strength=1500
    Pan Noise=true
    Pan Noise Field=&RandomRangeVector(Minimum=<-25 -25 -25>, Maximum=<25 25 25>)
   )
   &Drag(Drag=&RandomRangeFloat(Minimum=1, Maximum=5), Use Linear Drag=true, Use Rotational Drag=true) {enabled=false}
   &SolveForcesAndVelocity(Clamp Velocity=false, Limit Acceleration=false)
   &DynamicMaterialParameters(
    Index 0 Param 1=&FloatFromCurve(
     FloatCurve=DataInterface() {propertyValues {Curve {keys [
         {arriveTangent=-5.8452, interpMode=RCIM_Cubic, time=0, value=1}
         {arriveTangentWeight=0.0434, interpMode=RCIM_Cubic, leaveTangentWeight=0.217, tangentMode=RCTM_User, time=0.2, value=0}
         {arriveTangent=3.4164, interpMode=RCIM_Cubic, leaveTangent=3.4164, tangentMode=RCTM_User, time=0.8, value=0.7}
         <at 1 0 RCIM_Cubic>
        ]}}}
     Scale Curve=1.5
    )
    Index 0 Param 2=20
    Index 0 Param 3=250
    Index 0 Param 4=7
    Module.Param0WriteEnabled=true
    Module.Param1WriteEnabled=true
    Module.Param2WriteEnabled=true
    Module.Param3WriteEnabled=true
   )
  ]
  rendererClasses [&NiagaraSpriteRendererProperties]
  renderers [{rendererClass=&NiagaraSpriteRendererProperties}]
 }
 Goop_Tendrils {
  emitterUpdateScript [
   &EmitterState(Life Cycle Mode=::NewEnumerator1, Loop Behavior=::NewEnumerator1, Loop Duration Mode=::NewEnumerator1, UseLoopDelay=false)
   &SpawnBurst_Instantaneous(Spawn Group=5, Use Spawn Probability=false, UseLoopCountLimit=false) {module=SpawnBurst_Instantaneous005}
   &SpawnBurst_Instantaneous(Spawn Group=4, Use Spawn Probability=false, UseLoopCountLimit=false) {module=SpawnBurst_Instantaneous004}
   &SpawnBurst_Instantaneous(Spawn Group=3, Use Spawn Probability=false, UseLoopCountLimit=false) {module=SpawnBurst_Instantaneous003}
   &SpawnBurst_Instantaneous(Spawn Group=2, Use Spawn Probability=false, UseLoopCountLimit=false) {module=SpawnBurst_Instantaneous002}
   &SpawnBurst_Instantaneous(Spawn Group=1, Use Spawn Probability=false, UseLoopCountLimit=false) {module=SpawnBurst_Instantaneous001}
   &SpawnBurst_Instantaneous(Use Spawn Probability=false, UseLoopCountLimit=false)
  ]
  particleSpawnScript [
   &InitializeParticle(
    Lifetime=5
    Mesh Scale Max=<1 1 0.6>
    Mesh Scale Min=<0.5 0.5 0.2>
    Mesh Scale Mode=::NewEnumerator2
    UsePositionOffset=false
    Write Material Random=true
   )
   &InitialMeshOrientation(
    Mesh Axis To Orient=<0 0 1>
    Mesh Orientation Mode=::NewEnumerator2
    Orientation Vector=<0 1 0>
    Rotation=&RandomRangeVector(Minimum=<0 0 0>, Maximum=<0 0 1>)
    Use Rotation Vector=true
   )
  ]
  particleUpdateScript [
   &ParticleState(Kill Particles When Lifetime Has Elapsed=false, Let Infinitely Lived Particles Die When Emitter Deactivates=true)
   &Timeline(
    Play=&Boolean_OrOperation(
     Bool1=&SetBoolByFloatComparison(A=$User.ForwardDot, B=0.0001)
     Bool2=&SetBoolByFloatComparison(A=&Abs_Float(Float:NiagaraFloat=$User.RightDot), B=0.0001)
    )
    Rewind Rate=5
   )
   &SkeletalMeshLocation(
    Bone / Socket Index=&SpawnGroup()
    MaskSpawnGroup=false
    Mesh Sampling Type=::NewEnumerator5
    Skeletal Mesh=$User.GliderMesh
    Socket Sampling Mode=::NewEnumerator5
   )
   &ScaleMeshSize(Scale Factor=&MakeVector(
     X=1
     Y=1
     Z=&FloatFromCurve(
      FloatCurve=DataInterface() {propertyValues {Curve {keys [{arriveTangent=3.9082, interpMode=RCIM_Cubic, leaveTangent=3.9082, tangentMode=RCTM_User, time=0, value=0}, <at 1 1 RCIM_Cubic>]}}}
      CurveIndex=$StackContext.Timeline.NormalizedTime
     )
    ))
   &DynamicMaterialParameters(
    Index 0 Param 1=&Waveform("Waveform [1]"=::NewEnumerator12, "[1] Frequency"=15, "[1] Amplitude Min/Max"=<-1 1>, "[1] Scale"=0.125)
    Index 0 Param 2=&Waveform("Waveform [1]"=::NewEnumerator12, "[1] Frequency"=10, "[1] Amplitude Min/Max"=<-1 1>, "[1] Scale"=0.125)
    Index 0 Param 3=&Waveform(
     "Waveform [1]"=::NewEnumerator3
     "[1] Frequency"=&RandomRangeFloat(Minimum=1, Maximum=2)
     "[1] Amplitude Min/Max"=<0.6 2>
    )
    Index 0 Param 4=&Lerp_Float(A=-150, B=50, Alpha=$Particles.Timeline.NormalizedTime)
    Index 1 Param 1=-1
    Index 1 Param 2=2
    Index 1 Param 3=1
    Index 1 Param 4=&ClampFloat(
     Float=&RemapRange(
      Input Value To Remap=&Abs_Float(Float:NiagaraFloat=$User.ForwardDot)
      Input Min=0
      Input Max=1
      Output Min=0
      Output Max=2
     )
     Max=2
    )
    Module.Param0WriteEnabled=true
    Module.Param0WriteEnabled1=true
    Module.Param1WriteEnabled=true
    Module.Param1WriteEnabled1=true
    Module.Param2WriteEnabled=true
    Module.Param2WriteEnabled1=true
    Module.Param3WriteEnabled=true
    Module.Param3WriteEnabled1=true
    Write Parameter Index 1=true
   )
  ]
  rendererClasses [&NiagaraMeshRendererProperties]
  renderers [{rendererClass=&NiagaraMeshRendererProperties, rendererIndex=0}]
 }
 Speedlines {
  simTarget=GPUComputeSim
  emitterUpdateScript [
   &EmitterState()
   &SpawnRate(
    SpawnRate=&MakeCustomFloatFromBool(
     Bool=&Boolean_AndOperation(Bool1=$User.bIsFrontEndPreview, Bool2=$User.bIsFullyDeployed)
     True Float=50
     False Float=0
    )
    Use Spawn Probability=false
   )
   &SpawnPerUnit(
    Module.Use Max Movement Threshold=true
    Module.Use Movement Tolerance=true
    Movement Tolerance=10
    Spawn Spacing:NiagaraFloat=&MakeCustomFloatFromBool(
     Bool=$User.bIsFullyDeployed
     True Float=&MakeCustomFloatFromBool(
      Bool=&SetBoolByFloatComparison(A=&Abs_Float(Float:NiagaraFloat=$User.ForwardDot), B=0.1, Comparison Type=::NewEnumerator2)
      True Float=40
      False Float=0
     )
     False Float=0
    )
    Use Spawn Probability=false
   )
  ]
  particleSpawnScript [
   &InitializeParticle(
    Color=&MakeLinearColorFromVectorAndFloat(
     "Vector (RGB)"=&MakeVectorFromLinearColorRGB(Color=<rgba 0.2292 0.1989 0.1838 1>)
     "Float (Alpha)"=&RandomRangeFloat(Minimum=0.125, Maximum=0.333)
    )
    Lifetime Max=0.4
    Lifetime Min=0.2
    Lifetime Mode=::NewEnumerator1
    Sprite Size Max=<5 4>
    Sprite Size Min=<3 2>
    Sprite Size Mode=::NewEnumerator2
    UsePositionOffset=false
    Write Material Random=true
   )
   &ShapeLocation(
    Handle Radius=75
    Large Radius=200
    MaskSpawnGroup=false
    Override Seed=false
    Rotation Axis=<0 1 0>
    Rotation Mode=::NewEnumerator1
    Shape Primitive=::NewEnumerator3
    UseOwnerScale=true
   )
   &TorusLocation(
    Align Torus With Axis=true
    Coordinate Space:ENiagaraCoordinateSpace=::Simulation
    Handle Radius=75
    Large Radius=200
    MaskSpawnGroup=false
    Module.Override Randomness=false
    Torus Axis=&TransformVector(Vector=<0 1 0>, Destination Space=::Simulation)
   ) {enabled=false}
   &InheritVelocity(Module.Limit Speed=true, Velocity Scale=<0.7 0.7 0>)
   &AddVelocity(Velocity=<0 1 0>, Velocity Speed Scale=500)
  ]
  particleUpdateScript [
   &ParticleState()
   &SolveForcesAndVelocity(Clamp Velocity=false, Limit Acceleration=false)
   &ScaleSpriteSize(Scale Factor=&Vector2DFromCurve(
     Vector2Curve=DataInterface() {propertyValues {
       XCurve {keys [<at -0.0069 0.5 RCIM_Cubic>,<at 0.4918 1 RCIM_Cubic>,<at 1 0.5 RCIM_Cubic>]}
       YCurve {keys [<at 0 0.0245 RCIM_Cubic>, {arriveTangent=-0.0245, interpMode=RCIM_Cubic, leaveTangent=-0.0245, time=0.4975, value=1}, <at 1 0 RCIM_Cubic>]}
      }}
     ScaleCurve=<1 2>
    ))
   &ScaleSpriteSizeBySpeed(Max Scale Factor=<1 100>, Min Scale Factor=<1 10>)
  ]
  rendererClasses [&NiagaraSpriteRendererProperties]
  renderers [{rendererClass=&NiagaraSpriteRendererProperties}]
 }
 Wind_Cone {
  emitterUpdateScript [
   &EmitterState(
    Inactive Response=::NewEnumerator1
    Life Cycle Mode=::NewEnumerator1
    Loop Behavior=::NewEnumerator1
    Loop Duration Mode=::NewEnumerator1
    UseLoopDelay=false
   )
   &SpawnBurst_Instantaneous(Use Spawn Probability=false, UseLoopCountLimit=false)
  ]
  particleSpawnScript [
   &InitializeParticle(Lifetime=0, Mesh Scale=<2.5 2 2>, Mesh Scale Mode=::NewEnumerator4, UsePositionOffset=false, Write Material Random=true)
   &InitialMeshOrientation(Mesh Orientation Mode=::NewEnumerator2, Orientation Vector=<0 -1 0>, Use Rotation Vector=true)
  ]
  particleUpdateScript [
   &ParticleState(Kill Particles When Lifetime Has Elapsed=false, Let Infinitely Lived Particles Die When Emitter Deactivates=true)
   &UpdateMeshOrientation(
    Facing Direction=&MakeCustomVectorFromBool(
     Bool=&SetBoolByFloatComparison(A=$User.ForwardDot, B=0, Comparison Type=::NewEnumerator2)
     True Vector=<1 0 0>
     False Vector=<-1 0 0>
    )
    Facing Interpolation=false
    Orientation Method=::NewEnumerator1
    Override Quat=false
    TransformFacing=false
    TransformUp=false
   )
   /SetVariables(Particles.Position=&MakeCustomVectorFromBool(
     Bool=&SetBoolByFloatComparison(A=$User.ForwardDot, B=0, Comparison Type=::NewEnumerator2)
     True Vector=&AddVectorToPosition(Position=$Particles.Initial.Position, Vector=<0 -100 0>)
     False Vector=&AddVectorToPosition(Position=$Particles.Initial.Position, Vector=<0 100 0>)
    )) {isSetParameters=true}
   &ScaleColor(
    Scale Alpha=&MakeCustomFloatFromBool(
     Bool=$User.bIsFrontEndPreview
     True Float=1
     False Float=&FloatFromCurve(
      FloatCurve=DataInterface() {propertyValues {Curve {keys [{arriveTangent=3.0636, interpMode=RCIM_Cubic, leaveTangent=3.0636, tangentMode=RCTM_User, time=0, value=0}, <at 1 1 RCIM_Cubic>]}}}
      CurveIndex=&Abs_Float(Float:NiagaraFloat=$User.ForwardDot)
     )
    )
    ScaleA=true
    ScaleRGB=true
   )
   &DynamicMaterialParameters(Write Parameter Index 0=false)
  ]
  rendererClasses [&NiagaraMeshRendererProperties]
  renderers [{rendererClass=&NiagaraMeshRendererProperties, rendererIndex=0}]
 }
}
summary {
 systemName=NS_Glider_BadBear
 emitters [
  {emitterName=Drips, rendererClasses [&NiagaraSpriteRendererProperties]}
  {emitterName=Wind_Cone, rendererClasses [&NiagaraMeshRendererProperties]}
  {emitterName=Speedlines, simTarget=GPUComputeSim, rendererClasses [&NiagaraSpriteRendererProperties]}
  {emitterName=Goop_Tendrils, rendererClasses [&NiagaraMeshRendererProperties]}
 ]
 userVariables [
  User.bIsFrontEndPreview {type=NiagaraBool}
  User.bIsFullyDeployed {type=NiagaraBool}
  User.ForwardDot {type=NiagaraFloat, defaultValue=0, description=""}
  User.GliderMesh {
   type=NiagaraDataInterfaceSkeletalMesh
   defaultValue {
    struct=NiagaraExt_VariableValue_DataInterface
    value {dataInterface=&self:NiagaraDataInterfaceSkeletalMesh_0}
   }
  }
  User.RightDot {type=NiagaraFloat, defaultValue=0, description=""}
 ]
}
