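# Niagara one-shot burst effect for deploying the BadBear glider, a cyan-tinted liquid splash with bubbles, goo, drops and sparks.
# A brief HDR flash with a light renderer joins an expanding, rotating bubble mesh paired with a glow sprite.
# Drip sparks, impact drops and spline mesh streaks launch from cylinder shapes under drag, gravity or noise.
# Curve-driven dynamic material parameters animate most layers; the straight-streak emitter is disabled.
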NiagaraSystem /BRCosmetics/Gadgets/Assets/VinderTech_GliderChute/Glider_BadBear/FX/NS_Glider_BadBear_Deploy
dependencies {
 usedDataInterfaces [&NiagaraDataInterfaceCurve, &NiagaraDataInterfaceVector2DCurve]
 usedDynamicInputs []
 usedModules [&SystemState]
}
emitters {
 Bubbles {
  emitterUpdateScript [
   &EmitterState()
   &SpawnBurst_Instantaneous(Use Spawn Probability=false, UseLoopCountLimit=false)
  ]
  particleSpawnScript [
   &InitializeParticle(
    Color=<rgba 1.3728 3.0599 5 0.5>
    Lifetime=0.5
    Mesh Scale Mode=::NewEnumerator3
    Mesh Uniform Scale=3.5
    Position Offset=<0 0 50>
    Sprite Size Mode=::NewEnumerator3
    Uniform Sprite Size=300
    UsePositionOffset=true
    Write Material Random=true
   )
   &AddVelocity(Velocity=<0 1 0>, Velocity Speed Scale=750) {enabled=false}
   &InitialMeshOrientation(
    Mesh Axis To Orient=<0 0 1>
    Mesh Orientation Mode=::NewEnumerator2
    Rotation=&RandomRangeVector(Minimum=<0 0 0>, Maximum=<1 1 1>)
    Use Rotation Vector=true
   )
   &CameraOffset(Camera Offset Amount=50)
  ]
  particleUpdateScript [
   &ParticleState()
   &Drag(Drag=5, Use Linear Drag=true, Use Rotational Drag=false)
   &GravityForce()
   &SolveForcesAndVelocity(Clamp Velocity=false, Limit Acceleration=false)
   &ScaleColor(
    Scale RGB=&VectorFromFloat(Value=&FloatFromCurve(FloatCurve=DataInterface() {propertyValues {Curve {keys [
          {arriveTangent=171.9012, interpMode=RCIM_Cubic, time=0, value=0}
          {arriveTangent=29.4118, interpMode=RCIM_Cubic, leaveTangent=29.4118, time=0.1, value=2}
          {arriveTangent=-89.7088, interpMode=RCIM_Cubic, leaveTangent=-89.7088, tangentMode=RCTM_User, time=0.17, value=5}
          {
           arriveTangent=-7.3461
           arriveTangentWeight=0.0565
           interpMode=RCIM_Cubic
           leaveTangent=-7.3461
           leaveTangentWeight=0.1797
           tangentMode=RCTM_User
           time=0.25
           value=1
          }
          {interpMode=RCIM_Cubic, tangentMode=RCTM_User, time=0.7, value=0}
         ]}}}))
    ScaleA=false
    ScaleRGB=true
   )
   &ScaleMeshSize(Scale Factor=&VectorFromFloat(Value=&FloatFromCurve(FloatCurve=DataInterface() {propertyValues {Curve {keys [{arriveTangent=10.7792, time=0, value=0}, {arriveTangent=0.5907, leaveTangent=0.5907, time=1, value=5.5}]}}})))
   &DynamicMaterialParameters(
    Index 0 Param 1=&FloatFromCurve(FloatCurve=DataInterface() {propertyValues {Curve {keys [
         {arriveTangent=-1.7239, interpMode=RCIM_Cubic, time=0.25, value=1}
         {arriveTangent=-4.0266, interpMode=RCIM_Cubic, leaveTangent=-4.0266, tangentMode=RCTM_User, time=1, value=0}
        ]}}})
    Index 0 Param 2=&RandomRangeFloat(Minimum=1.5, Maximum=2.5)
    Index 0 Param 3=&FloatFromCurve(
     FloatCurve=DataInterface() {propertyValues {Curve {keys [
         {arriveTangent=13.3756, interpMode=RCIM_Cubic, leaveTangent=13.3756, tangentMode=RCTM_User, time=0.4, value=0}
         <at 0.6 1 RCIM_Cubic>
         {arriveTangent=-4.0708, interpMode=RCIM_Cubic, leaveTangent=-4.0708, tangentMode=RCTM_User, time=1, value=0}
        ]}}}
     Scale Curve=20
    )
    Index 0 Param 4=&MakeFloatFromVector(Vector=$Particles.Scale, Channel=::NewEnumerator0)
    Index 3 Param 1=4
    Index 3 Param 2=0.4
    Module.Param0WriteEnabled=true
    Module.Param0WriteEnabled3=true
    Module.Param1WriteEnabled=true
    Module.Param1WriteEnabled3=true
    Module.Param2WriteEnabled=true
    Module.Param2WriteEnabled3=false
    Module.Param3WriteEnabled=true
    Module.Param3WriteEnabled3=false
    Write Parameter Index 3=true
   )
   &UpdateMeshOrientation(
    Rotation Rate=&FloatFromCurve(
     FloatCurve=DataInterface() {propertyValues {Curve {keys [
         {arriveTangent=10.9187, interpMode=RCIM_Cubic, time=0, value=1}
         {arriveTangent=-2.4372, interpMode=RCIM_Cubic, leaveTangent=-3.2804, tangentMode=RCTM_Break, time=0.4, value=0.5}
         <at 1 0>
        ]}}}
     Scale Curve=1.5
    )
    Rotation Vector=<0.1 1 0>
   )
   /SetVariables(Particles.GlowColor=&MakeLinearColorFromVectorAndFloat(
     "Vector (RGB)"=&MakeVectorFromLinearColorRGB(Color=<rgba 0.2448 0.6923 1 1>)
     "Float (Alpha)"=&FloatFromCurve(
      FloatCurve=DataInterface() {propertyValues {Curve {keys [
          {arriveTangent=18.2404, interpMode=RCIM_Cubic, leaveTangent=18.2404, tangentMode=RCTM_User, time=0, value=1}
          {arriveTangent=-1.2, interpMode=RCIM_Cubic, leaveTangent=-1.2, time=0.12, value=2}
          {
           arriveTangent=-4.84
           arriveTangentWeight=0.0565
           interpMode=RCIM_Cubic
           leaveTangent=-4.84
           leaveTangentWeight=0.1797
           tangentMode=RCTM_User
           time=0.25
           value=0.7
          }
          {arriveTangent=-6.1958, interpMode=RCIM_Cubic, leaveTangent=-6.1958, tangentMode=RCTM_User, time=0.4, value=0.9}
          {interpMode=RCIM_Cubic, tangentMode=RCTM_User, time=1, value=0}
         ]}}}
      Scale Curve=0.5
     )
    )) {isSetParameters=true}
   &ScaleSpriteSize(
    Uniform Curve Scale=&MakeFloatFromVector(Vector=$Particles.Scale, Channel=::NewEnumerator0)
    Uniform Curve Sprite Scale=DataInterface() {propertyValues {Curve {keys [
        {arriveTangent=171.9012, interpMode=RCIM_Cubic, time=0.4, value=0.3}
        {
         arriveTangent=-3.2031
         arriveTangentWeight=0.0565
         interpMode=RCIM_Cubic
         leaveTangent=-3.2031
         leaveTangentWeight=0.1797
         tangentMode=RCTM_User
         time=0.5
         value=0.7
        }
        {arriveTangent=-0.9894, interpMode=RCIM_Cubic, leaveTangent=-0.9894, tangentMode=RCTM_User, time=0.6, value=0.5}
        {arriveTangent=-2.7772, interpMode=RCIM_Cubic, leaveTangent=-2.7772, tangentMode=RCTM_User, time=1, value=0}
       ]}}}
   )
  ]
  rendererClasses [&NiagaraMeshRendererProperties, &NiagaraSpriteRendererProperties]
  renderers [
   {rendererClass=&NiagaraMeshRendererProperties, rendererIndex=0}
   {rendererClass=&NiagaraSpriteRendererProperties, rendererIndex=1}
  ]
 }
 Drip_Sparks {
  emitterUpdateScript [
   &EmitterState()
   &SpawnBurst_Instantaneous(
    Spawn Count=&RandomRangeInt(Minimum=16, Maximum=32)
    Use Spawn Probability=false
    UseLoopCountLimit=false
   )
  ]
  particleSpawnScript [
   &InitializeParticle(
    Color=<rgba 0 1.3276 2 1>
    Lifetime Max=1
    Lifetime Min=0.25
    Lifetime Mode=::NewEnumerator1
    Sprite Size Max=<8 14>
    Sprite Size Min=<4 10>
    Sprite Size Mode=::NewEnumerator2
    UsePositionOffset=false
    Write Material Random=true
   )
   &ShapeLocation(
    Cylinder Height=30
    MaskSpawnGroup=false
    Override Seed=false
    Rotation Axis=<0 1 0>
    Rotation Mode=::NewEnumerator1
    Shape Primitive=::NewEnumerator1
    Surface Only=true
    Use Endcaps In Surface Only Mode=false
    UseOwnerScale=true
   ) {module=ShapeLocation001}
   &AddVelocity(
    Module.Use Falloff Distance=false
    Override Seed=false
    Velocity Mode=::NewEnumerator1
    Velocity Speed=&RandomRangeFloat(Minimum=250, Maximum=1000)
   ) {module=AddVelocity001}
  ]
  particleUpdateScript [
   &ParticleState()
   &ScaleSpriteSize(Uniform Curve Sprite Scale=DataInterface() {propertyValues {Curve {keys [
        {arriveTangent=8.2127, interpMode=RCIM_Cubic, leaveTangent=8.2127, tangentMode=RCTM_User, time=0, value=0.5}
        {
         arriveTangent=4.2448
         arriveTangentWeight=0.232
         interpMode=RCIM_Cubic
         leaveTangent=4.2448
         leaveTangentWeight=2.2538
         tangentMode=RCTM_User
         time=0.07
         value=1
        }
        {arriveTangent=-2.5194, interpMode=RCIM_Cubic, leaveTangent=-2.5194, tangentMode=RCTM_User, time=0.75, value=0.8}
        {arriveTangent=-4.4982, interpMode=RCIM_Cubic, leaveTangent=-4.4982, tangentMode=RCTM_User, time=1, value=0}
       ]}}})
   &GravityForce(Gravity=&RandomRangeVector(Minimum=<0 0 -980>, Maximum=<0 0 -1600>))
   &VectorNoiseForce(
    Force Amount=&FloatFromCurve(FloatCurve=DataInterface() {propertyValues {Curve {keys [<at 0 1 RCIM_Cubic>,<at 1 0.1 RCIM_Cubic>]}}}, Scale Curve=2500)
    Module.Use Random Noise Time Interval=true
    Random Noise Time Interval:NiagaraFloat=&RandomRangeFloat(Minimum=0.1, Maximum=0.15)
   )
   &SolveForcesAndVelocity(Clamp Velocity=false, Limit Acceleration=false)
   &DynamicMaterialParameters(
    Index 0 Param 1=&FloatFromCurve(
     FloatCurve=DataInterface() {propertyValues {Curve {keys [
         {arriveTangent=-5.8452, interpMode=RCIM_Cubic, time=0, value=1}
         {arriveTangentWeight=0.0434, interpMode=RCIM_Cubic, leaveTangentWeight=0.217, tangentMode=RCTM_User, time=0.2, value=0}
         {arriveTangent=3.4164, interpMode=RCIM_Cubic, leaveTangent=3.4164, tangentMode=RCTM_User, time=0.8, value=0.7}
         <at 1 0 RCIM_Cubic>
        ]}}}
     Scale Curve=1.5
    )
    Index 0 Param 2=10
    Index 0 Param 3=350
    Index 0 Param 4=3
    Module.Param0WriteEnabled=true
    Module.Param1WriteEnabled=true
    Module.Param2WriteEnabled=true
    Module.Param3WriteEnabled=true
   )
  ]
  rendererClasses [&NiagaraSpriteRendererProperties]
  renderers [{rendererClass=&NiagaraSpriteRendererProperties}]
 }
 Flash {
  emitterUpdateScript [
   &EmitterState()
   &SpawnBurst_Instantaneous(Use Spawn Probability=false, UseLoopCountLimit=false)
  ]
  particleSpawnScript [
   &InitializeParticle(
    Color=<rgba 0 3.3191 5 1>
    Lifetime Max=0.1
    Lifetime Min=0.05
    Lifetime Mode=::NewEnumerator1
    Sprite Size Mode=::NewEnumerator1
    Uniform Sprite Size Max=160
    Uniform Sprite Size Min=125
    UsePositionOffset=false
    Write Material Random=true
   )
   &SubUVAnimation(End Frame=3, SubUV Animation Mode:ENiagara_SubUVLookupMode=::NewEnumerator2)
   &CameraOffset(Camera Offset Amount=150)
  ]
  particleUpdateScript [
   &ParticleState()
   &ScaleSpriteSize(Uniform Curve Sprite Scale=DataInterface() {propertyValues {Curve {keys [{arriveTangent=2.5475, interpMode=RCIM_Cubic, leaveTangent=2.5475, tangentMode=RCTM_User, time=0, value=0}, <at 1 1.5 RCIM_Cubic>]}}})
   &ScaleColor(
    Scale Alpha=&FloatFromCurve(FloatCurve=DataInterface() {propertyValues {Curve {keys [<at 0.75 1 RCIM_Cubic>, {arriveTangent=-9.2286, interpMode=RCIM_Cubic, leaveTangent=-9.2286, tangentMode=RCTM_User, time=1, value=0}]}}})
    Scale RGB=&VectorFromFloat(Value=&FloatFromCurve(FloatCurve=DataInterface() {propertyValues {Curve {keys [<at 0 100 RCIM_Cubic>, {arriveTangent=-961.3766, interpMode=RCIM_Cubic, leaveTangent=-961.3766, tangentMode=RCTM_User, time=0.25, value=1}]}}}))
    ScaleA=true
    ScaleRGB=true
   )
   /SetVariables(
    Particles.LightColor=&MultiplyLinearColor(
     Linear Color=$Particles.Initial.Color
     Scale Factor=&MakeFloat4FromVectorAndFloat(
      "Vector (RGB)"=<1 1 1>
      "Float (Alpha)"=&FloatFromCurve(
       FloatCurve=DataInterface() {propertyValues {Curve {keys [{interpMode=RCIM_Cubic, tangentMode=RCTM_User, time=0, value=1}, {interpMode=RCIM_Cubic, tangentMode=RCTM_User, time=1, value=0}]}}}
       Scale Curve=5
      )
     )
    )
    Particles.LightExponent=6
   ) {isSetParameters=true}
  ]
  rendererClasses [&NiagaraSpriteRendererProperties, &NiagaraLightRendererProperties]
  renderers [
   {rendererClass=&NiagaraSpriteRendererProperties}
   {rendererClass=&NiagaraLightRendererProperties, rendererIndex=1}
  ]
 }
 Goo_Bits {
  emitterUpdateScript [
   &EmitterState()
   &SpawnBurst_Instantaneous(Use Spawn Probability=false, UseLoopCountLimit=false)
  ]
  particleSpawnScript [
   &InitializeParticle(
    Lifetime Max=0.3
    Lifetime Min=0.05
    Lifetime Mode=::NewEnumerator1
    Sprite Size Mode=::NewEnumerator3
    Uniform Sprite Size=30
    UsePositionOffset=false
    Write Material Random=true
   )
   &SubUVAnimation(End Frame=3, SubUV Animation Mode:ENiagara_SubUVLookupMode=::NewEnumerator2)
   &CameraOffset(Camera Offset Amount=40)
  ]
  particleUpdateScript [
   &ParticleState()
   &DynamicMaterialParameters(
    Index 0 Param 1=&RandomRangeFloat(Minimum=0, Maximum=1)
    Index 0 Param 3=&FloatFromCurve(FloatCurve=DataInterface() {propertyValues {Curve {keys [{arriveTangent=2.6651, interpMode=RCIM_Cubic, leaveTangent=2.6651, tangentMode=RCTM_User, time=0, value=0}, <at 1 1 RCIM_Cubic>]}}})
    Index 0 Param 4=&FloatFromCurve(FloatCurve=DataInterface() {propertyValues {Curve {keys [<at 0 0 RCIM_Cubic>, {arriveTangent=2.332, interpMode=RCIM_Cubic, leaveTangent=2.332, tangentMode=RCTM_User, time=1, value=1}]}}})
    Module.Param0WriteEnabled=true
    Module.Param1WriteEnabled=false
    Module.Param2WriteEnabled=true
    Module.Param3WriteEnabled=true
   )
   &GravityForce(Gravity=<0 0 -1600>)
   &SolveForcesAndVelocity(Clamp Velocity=false, Limit Acceleration=false)
   &ScaleColor(Scale RGB=&VectorFromFloat(Value=3), ScaleA=true, ScaleRGB=true)
   &ScaleSpriteSize(
    Uniform Curve Scale=4
    Uniform Curve Sprite Scale=DataInterface() {propertyValues {Curve {keys [
        {interpMode=RCIM_Cubic, leaveTangent=8.4581, tangentMode=RCTM_Break, time=0, value=1}
        {arriveTangent=1.8771, interpMode=RCIM_Cubic, leaveTangent=1.8771, tangentMode=RCTM_User, time=0.3915, value=2.7119}
        {interpMode=RCIM_Cubic, tangentMode=RCTM_User, time=0.9983, value=3.3515}
       ]}}}
   )
  ]
  rendererClasses [&NiagaraSpriteRendererProperties]
  renderers [{rendererClass=&NiagaraSpriteRendererProperties}]
 }
 Impact_Drops {
  emitterUpdateScript [
   &EmitterState()
   &SpawnBurst_Instantaneous(
    Spawn Count=&RandomRangeInt(Minimum=8, Maximum=15)
    Use Spawn Probability=false
    UseLoopCountLimit=false
   )
  ]
  particleSpawnScript [
   &InitializeParticle(
    Lifetime Max=1
    Lifetime Min=0.1
    Lifetime Mode=::NewEnumerator1
    Sprite Size Max=<12 20>
    Sprite Size Mode=::NewEnumerator2
    UsePositionOffset=false
    Write Material Random=true
   )
   &ShapeLocation(
    Cylinder Height=50
    MaskSpawnGroup=false
    Override Seed=false
    Rotation Axis=<0 1 0>
    Rotation Mode=::NewEnumerator1
    Shape Primitive=::NewEnumerator1
    Surface Only=true
    Use Endcaps In Surface Only Mode=false
    UseOwnerScale=true
   ) {module=ShapeLocation001}
   &ShapeLocation(MaskSpawnGroup=false, Override Seed=false, Sphere Radius=10, Sphere Surface Distribution=1, UseOwnerScale=true)
   &AddVelocity(
    Module.Use Falloff Distance=false
    Override Seed=false
    Velocity Mode=::NewEnumerator1
    Velocity Speed=&RandomRangeFloat(Minimum=833.3333, Maximum=1666.6666)
   )
  ]
  particleUpdateScript [
   &ParticleState()
   &ScaleSpriteSize(
    Non-Uniform Curve Sprite Scale=DataInterface() {propertyValues {
      XCurve {keys [
        {arriveTangent=0.0516, interpMode=RCIM_Cubic, leaveTangent=0.0516, tangentMode=RCTM_User, time=0, value=2}
        {interpMode=RCIM_Cubic, tangentMode=RCTM_User, time=1, value=0.47}
       ]}
      YCurve {keys [
        {arriveTangent=-0.0866, interpMode=RCIM_Cubic, leaveTangent=-0.0866, tangentMode=RCTM_User, time=0, value=4}
        {interpMode=RCIM_Cubic, tangentMode=RCTM_User, time=0.9988, value=0.22}
       ]}
     }}
    Scale Sprite Size Mode=::NewEnumerator3
   )
   &GravityForce(Gravity=<0 0 -1600>)
   &Drag(Drag=7, Use Linear Drag=true, Use Rotational Drag=true)
   &SolveForcesAndVelocity(Clamp Velocity=false, Limit Acceleration=false)
   &ScaleSpriteSizeBySpeed(Max Scale Factor=<1 3>)
   &ScaleColor(
    Scale Alpha=&FloatFromCurve(
     FloatCurve=DataInterface() {propertyValues {Curve {keys [<at 0 1>,<at 0.7 1>,<at 1 0>]}}}
     Scale Curve=&RandomRangeFloat(Minimum=0.5, Maximum=1)
    )
    ScaleA=true
    ScaleRGB=true
   )
  ]
  rendererClasses [&NiagaraSpriteRendererProperties]
  renderers [{rendererClass=&NiagaraSpriteRendererProperties}]
 }
 Spline_Straight_Streaks {
  bEnabled=false
  emitterUpdateScript [
   &EmitterState()
   &SpawnBurst_Instantaneous(
    Spawn Count=&RandomRangeInt(Minimum=8, Maximum=13)
    Use Spawn Probability=false
    UseLoopCountLimit=false
   )
  ]
  particleSpawnScript [
   &InitializeParticle(
    Color=<rgba 0 1.3276 2 1>
    Lifetime Max=0.2
    Lifetime Min=0.1
    Lifetime Mode=::NewEnumerator1
    Mesh Scale Mode=::NewEnumerator3
    Mesh Uniform Scale=&MaskFloatBySpawnGroup(Masked Float=&RandomRangeFloat(Minimum=1, Maximum=1.2), Passthrough Float=0.5)
    UsePositionOffset=false
    Write Material Random=true
   )
   &ShapeLocation(
    Cylinder Height=30
    Cylinder Radius=50
    MaskSpawnGroup=false
    Override Seed=false
    Rotation Axis=<0 1 0>
    Rotation Mode=::NewEnumerator1
    Shape Primitive=::NewEnumerator1
    Surface Only=true
    Use Endcaps In Surface Only Mode=false
    UseOwnerScale=true
   ) {module=ShapeLocation001}
   &AddVelocity(
    Module.Use Falloff Distance=false
    Override Seed=false
    Velocity Mode=::NewEnumerator1
    Velocity Speed=&RandomRangeFloat(Minimum=750, Maximum=1500)
   ) {module=AddVelocity001}
   &CameraOffset(Camera Offset Amount=40)
  ]
  particleUpdateScript [
   &ParticleState()
   &ScaleMeshSize(Scale Factor=&VectorFromFloat(Value=&FloatFromCurve(FloatCurve=DataInterface() {propertyValues {Curve {keys [
          {arriveTangent=3.8261, interpMode=RCIM_Cubic, leaveTangent=3.8261, tangentMode=RCTM_User, time=0, value=0}
          {arriveTangent=0.7815, interpMode=RCIM_Cubic, leaveTangent=0.7815, tangentMode=RCTM_User, time=0.25, value=0.6}
          {arriveTangent=0.2614, interpMode=RCIM_Cubic, leaveTangent=0.2614, tangentMode=RCTM_User, time=1, value=1}
         ]}}})))
   &Drag(Drag=7, Use Linear Drag=true, Use Rotational Drag=true)
   &SolveForcesAndVelocity(Clamp Velocity=false, Limit Acceleration=false)
   &DynamicMaterialParameters(
    Index 0 Param 1=&FloatFromCurve(FloatCurve=DataInterface() {propertyValues {Curve {keys [<at 0 0 RCIM_Cubic>, {arriveTangent=0.9849, interpMode=RCIM_Cubic, leaveTangent=0.9849, tangentMode=RCTM_User, time=1, value=1}]}}})
    Index 0 Param 2=&FloatFromCurve(
     FloatCurve=DataInterface() {propertyValues {Curve {keys [
         {arriveTangent=1.3993, interpMode=RCIM_Cubic, leaveTangent=1.3993, tangentMode=RCTM_User, time=0, value=0.5}
         {interpMode=RCIM_Cubic, tangentMode=RCTM_User, time=1, value=1}
        ]}}}
     Scale Curve=1.5
    )
    Index 0 Param 3=&Multiply_Float(
     A=&MakeFloatFromVector(Vector=$Particles.Scale, Channel=::NewEnumerator0)
     B=&RandomRangeFloat(Minimum=3, Maximum=5)
    )
    Index 0 Param 4=&FloatFromCurve(
     FloatCurve=DataInterface() {propertyValues {Curve {keys [
         {interpMode=RCIM_Cubic, leaveTangentWeight=0.7017, tangentMode=RCTM_User, time=0, value=0}
         {arriveTangentWeight=1.2352, interpMode=RCIM_Cubic, tangentMode=RCTM_User, time=1, value=1}
        ]}}}
     Scale Curve=&RandomRangeFloat(Minimum=0.5, Maximum=1)
    )
    Index 1 Param 1=&FloatFromCurve(
     FloatCurve=DataInterface() {propertyValues {Curve {keys [
         {arriveTangent=2.4162, interpMode=RCIM_Cubic, leaveTangent=2.4162, tangentMode=RCTM_User, time=0, value=0}
         {interpMode=RCIM_Cubic, tangentMode=RCTM_User, time=1, value=1}
        ]}}}
     Scale Curve=-3
    )
    Index 1 Param 2=&FloatFromCurve(
     FloatCurve=DataInterface() {propertyValues {Curve {keys [{interpMode=RCIM_Cubic, tangentMode=RCTM_User, time=0.6, value=1}, {interpMode=RCIM_Cubic, tangentMode=RCTM_User, time=1, value=2}]}}}
     Scale Curve=&RandomRangeFloat(Minimum=0.5, Maximum=1.5)
    )
    Module.Param0WriteEnabled=true
    Module.Param0WriteEnabled1=true
    Module.Param1WriteEnabled=true
    Module.Param1WriteEnabled1=true
    Module.Param2WriteEnabled=true
    Module.Param2WriteEnabled1=false
    Module.Param3WriteEnabled=true
    Module.Param3WriteEnabled1=false
    Write Parameter Index 1=true
   )
  ]
  rendererClasses [&NiagaraMeshRendererProperties]
  renderers [{rendererClass=&NiagaraMeshRendererProperties, rendererIndex=0}]
 }
 Spline_Streaks {
  emitterUpdateScript [
   &EmitterState()
   &SpawnBurst_Instantaneous(
    Spawn Count=&RandomRangeInt(Minimum=12, Maximum=18)
    Use Spawn Probability=false
    UseLoopCountLimit=false
   )
  ]
  particleSpawnScript [
   &InitializeParticle(
    Color=<rgba 0 1.3276 2 1>
    Lifetime Max=0.4
    Lifetime Min=0.1
    Lifetime Mode=::NewEnumerator1
    Mesh Scale Mode=::NewEnumerator1
    Mesh Uniform Scale Max=1
    Mesh Uniform Scale Min=0.8
    UsePositionOffset=false
    Write Material Random=true
   )
   &ShapeLocation(
    Cylinder Height=50
    Cylinder Radius=50
    MaskSpawnGroup=false
    Override Seed=false
    Rotation Axis=<0 1 0>
    Rotation Mode=::NewEnumerator1
    Shape Primitive=::NewEnumerator1
    Surface Only=true
    Use Endcaps In Surface Only Mode=false
    UseOwnerScale=true
   ) {module=ShapeLocation001}
   &ShapeLocation(Hemisphere Distribution=<0.5 1>, MaskSpawnGroup=false, Override Seed=false, Sphere Radius=3, UseOwnerScale=true)
   &AddVelocity(
    Module.Use Falloff Distance=false
    Override Seed=false
    Velocity Mode=::NewEnumerator1
    Velocity Speed=&RandomRangeFloat(Minimum=250, Maximum=1000)
   )
  ]
  particleUpdateScript [
   &ParticleState()
   &UpdateMeshOrientation(
    Rotation Rate=&FloatFromCurve(
     FloatCurve=DataInterface() {propertyValues {Curve {keys [
         {arriveTangent=-2.1925, interpMode=RCIM_Cubic, leaveTangent=-2.1925, tangentMode=RCTM_User, time=0, value=1}
         {arriveTangent=-0.2845, interpMode=RCIM_Cubic, leaveTangent=-0.2845, tangentMode=RCTM_User, time=1, value=0.05}
        ]}}}
     Scale Curve=&RandomRangeFloat(Minimum=0.3, Maximum=0.4)
    )
    Rotation Vector=<0 1 -1>
   )
   &ScaleMeshSize(Scale Factor=&VectorFromFloat(Value=&FloatFromCurve(FloatCurve=DataInterface() {propertyValues {Curve {keys [
          {arriveTangent=3.8261, interpMode=RCIM_Cubic, leaveTangent=3.8261, tangentMode=RCTM_User, time=0, value=0}
          {arriveTangent=0.7815, interpMode=RCIM_Cubic, leaveTangent=0.7815, tangentMode=RCTM_User, time=0.25, value=0.6}
          {arriveTangent=0.2614, interpMode=RCIM_Cubic, leaveTangent=0.2614, tangentMode=RCTM_User, time=1, value=1}
         ]}}})))
   &GravityForce()
   &Drag(Drag=7, Use Linear Drag=true, Use Rotational Drag=true)
   &SolveForcesAndVelocity(Clamp Velocity=false, Limit Acceleration=false)
   &DynamicMaterialParameters(
    Index 0 Param 1=&FloatFromCurve(FloatCurve=DataInterface() {propertyValues {Curve {keys [<at 0 0 RCIM_Cubic>, {arriveTangent=0.9849, interpMode=RCIM_Cubic, leaveTangent=0.9849, tangentMode=RCTM_User, time=1, value=1}]}}})
    Index 0 Param 2=&FloatFromCurve(
     FloatCurve=DataInterface() {propertyValues {Curve {keys [
         {arriveTangent=1.3993, interpMode=RCIM_Cubic, leaveTangent=1.3993, tangentMode=RCTM_User, time=0, value=0.5}
         {interpMode=RCIM_Cubic, tangentMode=RCTM_User, time=1, value=1}
        ]}}}
     Scale Curve=1.5
    )
    Index 0 Param 3=&Multiply_Float(
     A=&MakeFloatFromVector(Vector=$Particles.Scale, Channel=::NewEnumerator0)
     B=&RandomRangeFloat(Minimum=3, Maximum=8)
    )
    Index 0 Param 4=&FloatFromCurve(
     FloatCurve=DataInterface() {propertyValues {Curve {keys [
         {interpMode=RCIM_Cubic, leaveTangentWeight=0.7017, tangentMode=RCTM_User, time=0, value=0}
         {arriveTangentWeight=1.2352, interpMode=RCIM_Cubic, tangentMode=RCTM_User, time=1, value=1}
        ]}}}
     Scale Curve=&RandomRangeFloat(Minimum=1, Maximum=2)
    )
    Index 1 Param 1=&FloatFromCurve(
     FloatCurve=DataInterface() {propertyValues {Curve {keys [
         {arriveTangent=2.4162, interpMode=RCIM_Cubic, leaveTangent=2.4162, tangentMode=RCTM_User, time=0, value=0}
         {interpMode=RCIM_Cubic, tangentMode=RCTM_User, time=1, value=1}
        ]}}}
     Scale Curve=-3
    )
    Index 1 Param 2=&FloatFromCurve(
     FloatCurve=DataInterface() {propertyValues {Curve {keys [{interpMode=RCIM_Cubic, tangentMode=RCTM_User, time=0.6, value=1}, {interpMode=RCIM_Cubic, tangentMode=RCTM_User, time=1, value=2}]}}}
     Scale Curve=&RandomRangeFloat(Minimum=1, Maximum=1.25)
    )
    Module.Param0WriteEnabled=true
    Module.Param0WriteEnabled1=true
    Module.Param1WriteEnabled=true
    Module.Param1WriteEnabled1=true
    Module.Param2WriteEnabled=true
    Module.Param2WriteEnabled1=false
    Module.Param3WriteEnabled=true
    Module.Param3WriteEnabled1=false
    Write Parameter Index 1=true
   )
  ]
  rendererClasses [&NiagaraMeshRendererProperties]
  renderers [{rendererClass=&NiagaraMeshRendererProperties, rendererIndex=0}]
 }
}
summary {
 systemName=NS_Glider_BadBear_Deploy
 emitters [
  {emitterName=Flash, rendererClasses [&NiagaraSpriteRendererProperties, &NiagaraLightRendererProperties]}
  {emitterName=Goo_Bits, rendererClasses [&NiagaraSpriteRendererProperties]}
  {emitterName=Spline_Streaks, rendererClasses [&NiagaraMeshRendererProperties]}
  {emitterName=Drip_Sparks, rendererClasses [&NiagaraSpriteRendererProperties]}
  {bEnabled=false, emitterName=Spline_Straight_Streaks, rendererClasses [&NiagaraMeshRendererProperties]}
  {emitterName=Impact_Drops, rendererClasses [&NiagaraSpriteRendererProperties]}
  {emitterName=Bubbles, rendererClasses [&NiagaraMeshRendererProperties, &NiagaraSpriteRendererProperties]}
 ]
 userVariables []
}
